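; Summer Event quest gameplay ability that works around triggered abilities not firing on aircraft exit, enabling delayed "NO Permit" VO after leaving the battle bus.
; On activation it ends immediately in competitive playlists, otherwise starts a 2.5-second timer.
; When the timer fires, it sends a quest stat event if the owning pawn has a valid movement-state tag, then ends either way.
; Ending the ability clears any valid timer.
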(event Ability|EventActivateAbility
  ;[1 1 1] Workaround to allow us to play NO Permit VO after a delay when exiting the battlebus (Triggered abilities can't be fired off of Aircraft Exit)
  (|CallParentFunction)
  (if (Class|FortGameStateAthena|IsCompetitivePlaylist (Fortnite|Athena|GameState|GetGameStateAthenafromContext (Variables|GAVars|Data|GetOwningPlayerPawn)))
    (Ability|EndAbility)
    (else
      (bind _returnvalue (Utilities|Time|SetTimerbyEvent (delegate Custom|PlayAfterBusVO) 2.5))
      (Variables|Default|SetTimerHandle _returnvalue))))

(event Custom|PlayAfterBusVO
  ;[1 1 1] Workaround to allow us to play NO Permit VO after a delay when exiting the battlebus (Triggered abilities can't be fired off of Aircraft Exit)
  (if (GameplayTags|HasAnyTags (GameplayTags|GetOwnedGameplayTags (Variables|GAVars|Data|GetOwningPlayerPawn) :!class /Script/GameplayTags.GameplayTagAssetInterface) (Variables|Default|GetValidMovementStates))
    (Quest|SendComplexCustomStatEvent (Variables|GAVars|Data|GetOwningPlayerQuestManager) 0 (Variables|Default|GetT_Quest) "(GameplayTags=)" 1)
    (Ability|EndAbility)
    (else
      (Ability|EndAbility))))

(event Ability|EventFailedToActivatePassiveAbility
  (params (ActorInfo :type Gameplay%20Ability%20Actor%20Info%20Structure))
  (|CallParentFunction))

(event Ability|EventOnEndAbility
  (params (bWasCancelled :type Boolean))
  (bind _timerhandle (Variables|Default|GetTimerHandle))
  (|CallParentFunction false)
  (if (Utilities|Time|IsValidTimerHandle _timerhandle)
    (Utilities|Time|ClearandInvalidateTimerbyHandle _timerhandle)))
